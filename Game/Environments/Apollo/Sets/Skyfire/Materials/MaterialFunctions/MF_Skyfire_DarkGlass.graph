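# Material function for Skyfire dark glass, setting attributes for a metallic, anisotropic, clear-coated surface.
# Packed masks and curve atlas rows drive color, glow and AO, with optional custom primitive data for emissive color.
# World-aligned textures vary roughness and metallic and add detail normals, alongside a nested reflection-based Chroma function.
# Optional emissive glyphs combine a time-animated tech pattern and inner glow, with UVs offset by object position.
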MaterialFunction /Game/Environments/Apollo/Sets/Skyfire/Materials/MaterialFunctions/MF_Skyfire_DarkGlass
graph {nodes [
  n1 = SetMaterialAttributes!(
   MaterialAttributes
   BaseColor=NamedRerouteUsage()
   Metallic=NamedRerouteUsage()
   Specular=ScalarParameter() {defaultValue=0.8, group=Specular, parameterName=Specular}
   Roughness=NamedRerouteUsage()
   Anisotropy=NamedRerouteUsage()
   EmissiveColor=NamedRerouteUsage()
   Normal=NamedRerouteUsage()
   Tangent=NamedRerouteUsage()
   CustomData0=ScalarParameter() {defaultValue=1, group=Clear Coat, parameterName=Clear Coat}
   CustomData1=ScalarParameter() {group=Clear Coat, parameterName=Clear Coat Roughness}
   AmbientOcclusion=NamedRerouteUsage()
  ) {attributeSetTypes [
    #1
    #2
    #3
    #4
    #5
    #6
    #7
    #8
    #9
    #10
    #11
   ]}
  FunctionOutput(None=@n1)
  n2 = NamedRerouteUsage()
  n3 = ScalarParameter() {defaultValue=800, group=World Mask, parameterName=WorldMaskDivisor}
  NamedRerouteDeclaration(None=TextureSample(UVs=Divide(A=@n2, B=@n3)) {samplerSource=SSM_Wrap_WorldGroupSettings, texture=/Game/Global/Textures/Default/Noise/T_Tiling_Noise_01}.R) {name=WorldMask, nodeColor=<rgba 0 1 0.6016 1>}
  n4 = TextureSampleParameter2D() {parameterName=Masks, samplerType=SAMPLERTYPE_LinearColor, texture=/Game/Environments/Apollo/Sets/Skyfire/Textures/T_rearwall_MSK}
  NamedRerouteDeclaration(None=@n4.RGB) {name=MasksRGB, nodeColor=<rgba 0 0.0391 1 1>}
  n5 = CurveAtlasRowParameter(CurveTime=NamedRerouteUsage()) {
   atlas=/Game/Environments/Apollo/Sets/Skyfire/Curves/SkyfireCurveAtlas
   curve=/Game/Environments/Apollo/Sets/Skyfire/Curves/C_Skyfire_DarkColor_Diffuse
   defaultValue=3
   group=BaseColor
   parameterName=DiffuseColorCurve
   primitiveDataIndex=1
  }
  NamedRerouteDeclaration(None=Multiply(A=NamedRerouteUsage(), B=@n5)) {name=Base Color, nodeColor=<rgba 1 0 0.6328 1>}
  n6 = NamedRerouteUsage()
  n7 = Ceil(None=NamedRerouteUsage())
  n8 = CurveAtlasRowParameter(CurveTime=NamedRerouteUsage()) {
   atlas=&SkyfireCurveAtlas
   curve=/Game/Environments/Apollo/Sets/Skyfire/Curves/C_Skyfire_Black
   defaultValue=9
   group=Emissive
   parameterName=GlowColorCurve
  }
  n9 = NamedRerouteUsage()
  n10 = CurveAtlasRowParameter() {
   atlas=&SkyfireCurveAtlas
   bUseCustomPrimitiveData=true
   curve=&C_Skyfire_DarkColor_Diffuse
   defaultValue=3
   group=BaseColor
   parameterName=EmissiveColorCurve
   primitiveDataIndex=1
  }
  n11 = CurveAtlasRowParameter() {
   atlas=&SkyfireCurveAtlas
   curve=&C_Skyfire_Black
   defaultValue=9
   group=Emissive
   parameterName=GlowColorCurve
  }
  n12 = StaticSwitchParameter(True=@n10, False=@n11) {group=Emissive, parameterName=EmissiveUsesCPD}
  n13 = Add(
   A=NamedRerouteUsage()
   B=StaticSwitchParameter(True=Multiply(A=@n7, B=@n8), False=Multiply(A=@n9, B=@n12)) {group=Emissive, parameterName=EmissiveUses_ReMap_Red_ChannelToColorCurve}
  )
  NamedRerouteDeclaration(None=StaticSwitchParameter(True=Add(A=@n6, B=@n13), False=@n13) {parameterName=UseGlyphsEmissive}) {name=Emissive, nodeColor=<rgba 1 0.7969 0 1>}
  n14 = TextureSampleParameter2D(UVs=Divide(A=NamedRerouteUsage(), B=ScalarParameter() {defaultValue=80, parameterName=RoughnessMetallicMaskWorldTileDivisor})) {
   parameterName=PackB
   samplerSource=SSM_Wrap_WorldGroupSettings
   samplerType=SAMPLERTYPE_LinearColor
   texture=/Game/Environments/Apollo/Sets/Skyfire/Textures/T_Skyfire_PackB
  }
  NamedRerouteDeclaration(None=LinearInterpolate(
    A=ScalarParameter() {defaultValue=0.22, group=Roughness, parameterName=RoughnessLow}
    B=ScalarParameter() {defaultValue=0.25, group=Roughness, parameterName=RoughnessHigh}
    Alpha=@n14.R
   )) {name=Roughness, nodeColor=<rgba 0 0.9297 1 1>}
  NamedRerouteDeclaration(None=LinearInterpolate(
    A=ScalarParameter() {defaultValue=0.8, group=Metallic, parameterName=MetallicLow}
    B=ScalarParameter() {defaultValue=0.7, group=Metallic, parameterName=MetallicHigh}
    Alpha=@n14.R
   )) {name=Metallic, nodeColor=<rgba 1 0.1172 0 1>}
  n15 = TextureCoordinate()
  n16 = ObjectPositionWS()
  n17 = Constant3Vector() {constant=<rgba 28 55.106 875 1>}
  n18 = Multiply(A=ComponentMask(None=Fmod(A=@n16, B=@n17)) {g=true, r=true}, B=Constant2Vector() {g=1})
  n19 = MaterialFunctionCall(Float2=MaterialFunctionCall(UVs=Add(A=@n15, B=@n18), Texture Scale=VectorParameter() {parameterName=GlyphsUVCenterScale, defaultValue=<rgba 0.5 4 0 0>}.RGB) {materialFunction=&ScaleUVsByCenter}.UVs) {materialFunction=&BreakOutFloat2Components}
  n20 = TextureSampleParameter2D(UVs=MaterialFunctionCall(X=Saturate(None=@n19.R), Y=@n19.G) {materialFunction=&MakeFloat2}.Result) {
   parameterName=Glyphs
   samplerSource=SSM_Wrap_WorldGroupSettings
   samplerType=SAMPLERTYPE_LinearColor
   texture=/Game/Environments/Apollo/Sets/Skyfire/Textures/T_Skyfire_Glyphs
   channelNames {a="", b=Unused, g=Inner Glow, r=Base Glyph}
  }
  n21 = TextureSample(UVs=Multiply(A=TextureCoordinate(), B=ScalarParameter() {defaultValue=2, parameterName=GlyphTechPatternUVScale})) {
   samplerSource=SSM_Wrap_WorldGroupSettings
   samplerType=SAMPLERTYPE_LinearColor
   texture=/Game/Environments/Apollo/Sets/Skyfire/Textures/T_Skyfire_PackA
  }
  n22 = ScalarParameter() {defaultValue=4, group=Glyphs, parameterName=GlyphTechPatternTimeDivisor}
  n23 = ScalarParameter() {defaultValue=1, group=Glyphs, parameterName=GlyphTechPatternTimeMod}
  n24 = Sine(None=Add(
    A=@n21.G
    B=MaterialFunctionCall!(Time, Divisor=@n22, Mod=@n23) {materialFunction=/Game/Effects/Fort_Effects/Materials/_Functions/MF_ModuloTime}.Unclamped
   ))
  n25 = ScalarParameter() {defaultValue=1.5, group=Glyphs, parameterName=GlyphTechPatternBias}
  n26 = Multiply(
   A=@n20.Base Glyph
   B=Multiply(A=Add(A=@n24, B=@n25), B=ScalarParameter() {defaultValue=1, group=Glyphs, parameterName=GlyphTechPatternScale})
  )
  n27 = Power(
   Base=Multiply(A=@n20.Inner Glow, B=ScalarParameter() {defaultValue=1.4, group=Glyphs, parameterName=GlyphScale})
   Exp=ScalarParameter() {defaultValue=4, group=Glyphs, parameterName=GlyphsExponent}
  )
  n28 = CurveAtlasRowParameter() {
   atlas=&SkyfireCurveAtlas
   bUseCustomPrimitiveData=true
   curve=&C_Skyfire_DarkColor_Diffuse
   defaultValue=3
   group=BaseColor
   parameterName=EmissiveColorCurve
   primitiveDataIndex=1
  }
  n29 = VectorParameter() {group=Emissive, parameterName=GlyhpsColor, defaultValue=<rgba 0.3846 0.0385 0.4 0>}
  NamedRerouteDeclaration(None=Multiply(A=Add(A=@n26, B=@n27), B=StaticSwitchParameter(True=@n28, False=@n29.RGB) {group=Emissive, parameterName=EmissiveUsesCPD})) {name=Glyphs, nodeColor=<rgba 0 1 0.1094 1>}
  n30 = ScalarParameter() {defaultValue=1, group=Ambient Occlusion, parameterName=AO_Strength}
  n31 = NamedRerouteUsage()
  NamedRerouteDeclaration(None=LinearInterpolate(B=NamedRerouteUsage(), Alpha=Max(A=@n30, B=@n31)) {constA=1}) {name=AmbientOcclusion, nodeColor=<rgba 1 0.7969 0 1>}
  NamedRerouteDeclaration(None=@n4.R) {name=MasksR_Glow, nodeColor=<rgba 0.1016 1 0 1>}
  NamedRerouteDeclaration(None=MaterialFunctionCall() {materialFunction=/Game/Packages/Fortress_SharedMaterials/Base_Material_Functions/MF_Curie_WorldAlignedTexture}.Result) {name=CheapWorldAlignedUVs, nodeColor=<rgba 0 0.7656 1 1>}
  NamedRerouteDeclaration(None=Multiply(A=ScalarParameter() {defaultValue=0.25, group=Anisotropy, parameterName=Anisotropy}, B=NamedRerouteUsage())) {name=Anisotropy, nodeColor=<rgba 0.5 0 1 1>}
  n32 = TextureSampleParameter2D(UVs=StaticSwitchParameter(True=TextureCoordinate() {coordinateIndex=1}, False=TextureCoordinate()) {group=Normal, parameterName=NormalMapUseUV1}) {
   group=Normal
   parameterName=Normal
   samplerSource=SSM_Wrap_WorldGroupSettings
   samplerType=SAMPLERTYPE_Normal
   texture=/Game/Environments/Apollo/Sets/Skyfire/Textures/T_rear_wall_N
  }
  n33 = TextureSampleParameter2D(UVs=Divide(A=NamedRerouteUsage(), B=ScalarParameter() {defaultValue=80, group=Normal, parameterName=DetailNormalTileDivisor})) {
   group=Normal
   parameterName=DetailedNormal
   samplerSource=SSM_Wrap_WorldGroupSettings
   samplerType=SAMPLERTYPE_Normal
   texture=/Game/Global/Textures/Default/Tileable_Patterns/T_Skyfire_DarkGlass_DetailNormal
  }
  n34 = NamedRerouteDeclaration(None=@n33.RGB) {name=DetailedNormal, nodeColor=<rgba 0.8516 1 0 1>}
  n35 = VectorParameter() {group=Normal, parameterName=DetailNormalStrength, defaultValue=<rgba 0.02 0.02 1 0>}
  NamedRerouteDeclaration(None=MaterialFunctionCall(BaseNormal=@n32.RGB, AdditionalNormal=Multiply(A=@n34, B=@n35.RGB)) {materialFunction=&BlendAngleCorrectedNormals}.Result) {name=FinalNormals, nodeColor=<rgba 1 0 0.5859 1>}
  NamedRerouteDeclaration(None=@n4.G) {name=MasksG_Height, nodeColor=<rgba 0 1 0 1>}
  NamedRerouteDeclaration(None=@n4.B) {name=MasksB_AO, nodeColor=<rgba 0 0 1 1>}
  n36 = Normalize(VectorInput=Transform(None=NamedRerouteUsage()))
  NamedRerouteDeclaration(None=MaterialFunctionCall!(
    AmbientOcclusionMask=NamedRerouteUsage()
    AmbientOcclusionExponent
    ColorMask
    ReflectionVector=ReflectionVectorWS(CustomWorldNormal=@n36)
    FresnelMaskExponent=Constant() {r=2}
   ) {materialFunction=&/MF_Skyfire_Chroma}.Result) {name=Chroma, nodeColor=<rgba 0.5 0 1 1>}
  NamedRerouteDeclaration(None=@n4.A) {name=MasksA_CurvatureFine, nodeColor=<rgba 1 1 1 1>}
 ]}
