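# Parent material for dark glass in the Apollo Skyfire environment set, using Default Lit shading and a Metal physical material.
# MF_Skyfire_DarkGlass builds the full material attributes.
# The UseValetDamage static switch optionally routes that result through the vehicle damage function MF_Valet_Damage.
# It supports Nanite, instanced static meshes and dithered LOD transitions.
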Material /Game/Environments/Apollo/Sets/Skyfire/Materials/Parents/M_Skyfire_DarkGlass_Parent
graph {
 properties {
  bUseMaterialAttributes=true
  bUsedWithInstancedStaticMeshes=true
  bUsedWithNanite=true
  ditheredLODTransition=true
  physMaterial=/Game/FortressPhysicalMaterials/Metal
  pixelDepthOffsetMode=PDOM_Legacy
  usedShadingModels=Default Lit
 }
 nodes [
  n1 = MaterialFunctionCall() {materialFunction=/Game/Environments/Apollo/Sets/Skyfire/Materials/MaterialFunctions/MF_Skyfire_DarkGlass}
  StaticSwitchParameter(
   True=MaterialFunctionCall!(In=@n1.Result, Metal Tint, Local Bounds Split) {materialFunction=/Game/Vehicles/Valet/MF_Valet_Damage}.Result
   False=@n1.Result
  ) {parameterName=UseValetDamage}
 ]
}
