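# M_Skyfire_Energy_Parent is a parent material for energy surfaces in the Apollo Skyfire environment set.
# Its only node is a call to the MF_Skyfire_Energy material function, so the shading logic lives in that function rather than in this graph.
# It uses material attributes with Default Lit shading and enables usage on instanced static, skeletal, Nanite and mesh-deformer meshes.
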Material /Game/Environments/Apollo/Sets/Skyfire/Materials/Parents/M_Skyfire_Energy_Parent
graph {
 properties {
  bUseMaterialAttributes=true
  bUsedWithInstancedStaticMeshes=true
  bUsedWithMeshDeformer=true
  bUsedWithNanite=true
  bUsedWithSkeletalMesh=true
  pixelDepthOffsetMode=PDOM_Legacy
  usedShadingModels=Default Lit
 }
 nodes [MaterialFunctionCall() {materialFunction=/Game/Environments/Apollo/Sets/Skyfire/Materials/MaterialFunctions/MF_Skyfire_Energy}]
}
