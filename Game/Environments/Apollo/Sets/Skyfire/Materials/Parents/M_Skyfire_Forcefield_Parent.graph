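# Parent material for Skyfire set forcefields: additive, two-sided and lit, with an optional custom-primitive-data curve that hue-shifts the emissive.
# Panned dust and caustic noise, sine-animated grain, Fresnel, reflection-vector rainbow chroma and depth-based intersection edges build the look.
# UVs switch between mesh and world-aligned projection.
# Distance fade, biased by view angle, drives opacity and WPO culling that collapses vertices to the pivot, optionally masked by vertex color.
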Material /Game/Environments/Apollo/Sets/Skyfire/Materials/Parents/M_Skyfire_Forcefield_Parent
graph {
 properties {
  bUsedWithMeshDeformer=true
  bUsedWithSkeletalMesh=true
  blendMode=BLEND_Additive
  pixelDepthOffsetMode=PDOM_Legacy
  translucencyLightingMode=TLM_VolumetricPerVertexDirectional
  twoSided=true
  usedShadingModels=Default Lit
 }
 nodes [
  n1 = CurveAtlasRowParameter() {
   atlas=/Game/Environments/Apollo/Sets/Skyfire/Curves/SkyfireCurveAtlas
   bUseCustomPrimitiveData=true
   curve=/Game/Environments/Apollo/Sets/Skyfire/Curves/C_Skyfire_Black
   defaultValue=9
   group=Chroma
   parameterName=LootColorCurveCPD
   primitiveDataIndex=4
  }
  n2 = NamedRerouteUsage()
  n3 = LinearInterpolate(
   A=ScalarParameter() {defaultValue=0.25, group=Fresnel, parameterName=FresnelLow}
   B=ScalarParameter() {defaultValue=1, group=Fresnel, parameterName=FresnelHigh}
   Alpha=MaterialFunctionCall(Exponent=ScalarParameter() {defaultValue=1, group=Fresnel, parameterName=FresnelExponent}) {materialFunction=&MF_Fresnel_FX}.Fresnel Inverted Double Sided
  )
  n4 = NamedRerouteUsage()
  n5 = VectorParameter() {group=Color, parameterName=ColorBaseA, defaultValue=<rgba 0.3038 0.0702 0.1097 0>}
  n6 = VectorParameter() {group=Color, parameterName=ColorBaseB, defaultValue=<rgba 0.6163 0.1352 0.2487 0>}
  n7 = NamedRerouteUsage()
  n8 = Multiply(
   A=Add(A=Add(A=@n2, B=@n3), B=Multiply(A=@n3, B=@n4))
   B=LinearInterpolate(
    A=LinearInterpolate(A=@n5.RGB, B=@n6.RGB, Alpha=@n7)
    B=VectorParameter() {group=Color, parameterName=ColorSparkles, defaultValue=<rgba 0.9253 0.1919 0.8751 0>}.RGB
    Alpha=NamedRerouteUsage()
   )
  )
  n9 = StaticSwitchParameter(
   True=LinearInterpolate(A=@n8, B=VectorParameter() {group=Color, parameterName=ColorEdges, defaultValue=<rgba 25 1.6998 13.4256 0>}.RGB, Alpha=NamedRerouteUsage())
   False=@n8
  ) {defaultValue=true, group=IntersectionEdge, parameterName=UseIntersectionEdge}
  Multiply(
   A=StaticSwitchParameter(
    True=MaterialFunctionCall(Hue Shift Percentage=@n1, Texture=@n9) {materialFunction=&HueShift}.Result
    False=@n9
   ) {group=Color, parameterName=EmissiveUsesCPD}
   B=ScalarParameter() {defaultValue=0.12, parameterName=EmissiveBoost}
  )
  n10 = VertexColor()
  n11 = ChannelMaskParameter(None=AppendVector(A=@n10, B=@n10)) {group=DistanceFade+VertCulling, maskChannel=Alpha, parameterName=VertexColorMaskForDistanceFade+VertCulling, defaultValue=<rgba 0 0 0 1>}
  n12 = Ceil(None=NamedRerouteDeclaration(None=@n11.R) {name=VertexColorMaskForDistanceFade+VertCulling, nodeColor=<rgba 0 0.625 1 1>})
  n13 = MaterialFunctionCall() {materialFunction=&ObjectPivotPoint}
  n14 = DotProduct(
   A=ObjectOrientation()
   B=MaterialFunctionCall() {materialFunction=&CameraDirectionVector}.Result
  )
  n15 = Multiply(
   A=OneMinus(None=Abs(None=@n14))
   B=ScalarParameter() {defaultValue=1000, group=DistanceFade+VertCulling, parameterName=OrthogonalAngleDistanceBias}
  )
  n16 = Add(A=ScalarParameter() {defaultValue=1000, group=DistanceFade+VertCulling, parameterName="FadeEnd&&CullDistance"}, B=@n15)
  n17 = MaterialFunctionCall(
   Position=@n13.Object Pivot Location
   Fade_Start_Distance=@n16
   Fade_End_Distance=Add(A=ScalarParameter() {defaultValue=700, group=DistanceFade+VertCulling, parameterName=FadeStartDistance}, B=@n15)
  ) {materialFunction=&MF_Distance_SyncedFade_FX}
  n18 = Ceil(None=NamedRerouteDeclaration(None=@n17.Result Inverted) {name=DistanceFadeResult, nodeColor=<rgba 1 0 0.6562 1>})
  n19 = Multiply(
   A=Subtract(
    A=MaterialFunctionCall() {materialFunction=&ObjectPivotPoint}.Object Pivot Location
    B=WorldPosition().XYZ
   )
   B=Multiply(A=OneMinus(None=@n12), B=OneMinus(None=@n18))
  )
  n20 = If(A=@n16, B=@n17.Distance, "A > B"=Constant(), "A < B"=Subtract(A=@n13.Object Pivot Location, B=WorldPosition().XYZ))
  NamedRerouteDeclaration(None=StaticSwitchParameter(
    True=StaticSwitchParameter(True=@n19, False=@n20) {group=DistanceFade+VertCulling, parameterName=UseVertexColorForDistanceFade+VertCulling}
    False=Constant()
   ) {defaultValue=true, group=DistanceFade+VertCulling, parameterName=UseDistanceFade+VertCulling}) {name=WPO, nodeColor=<rgba 0.5937 1 0 1>}
  n21 = Divide(A=NamedRerouteUsage(), B=ScalarParameter() {defaultValue=3500, group=Normal, parameterName=NormalUVDivisor})
  NamedRerouteDeclaration(None=Multiply(
    A=TextureSample(UVs=@n21) {samplerType=SAMPLERTYPE_Normal, texture=/Game/Global/Textures/Default/Noise/T_SwirlNoise_N}.RGB
    B=VectorParameter() {group=Normal, parameterName=NormalStrength, defaultValue=<rgba 0.4 0.4 1 0>}.RGB
   )) {name=Normal, nodeColor=<rgba 0.5 0.5 1 1>}
  n22 = ScalarParameter() {defaultValue=4, group=IntersectionEdge, parameterName=IntersectEdgeDivisor}
  NamedRerouteDeclaration(None=Power(
    Base=MaterialFunctionCall!(Distance Bias, Distance Divisor=@n22, Scene Depth, Pixel Depth) {materialFunction=/Game/Effects/Fort_Effects/Materials/_Functions/MF_DistanceBehindTranslucency_01}.Normalized Inverted
    Exp=ScalarParameter() {defaultValue=4, group=IntersectionEdge, parameterName=IntersectionEdgeExponent}
   ) {constExponent=4}) {name=IntersectionEdges, nodeColor=<rgba 1 0.1172 0 1>}
  n23 = NamedRerouteUsage()
  n24 = LinearInterpolate(A=NamedRerouteUsage(), Alpha=@n23)
  n25 = StaticSwitchParameter(
   True=MaterialFunctionCall(Distance=ScalarParameter() {defaultValue=150, group=NearCameraFade, parameterName=NearCameraFadeDistance}) {materialFunction=&Near_Camera_Fade}.Result
   False=Constant() {r=1}
  ) {defaultValue=true, group=NearCameraFade, parameterName=UseNearCameraFade}
  n26 = Multiply(A=@n25, B=@n23)
  NamedRerouteDeclaration(None=StaticSwitchParameter(
    True=StaticSwitchParameter(True=@n24, False=@n26) {group=DistanceFade+VertCulling, parameterName=UseVertexColorForDistanceFade+VertCulling}
    False=@n25
   ) {defaultValue=true, group=DistanceFade+VertCulling, parameterName=UseDistanceFade+VertCulling}) {name=Opacity, nodeColor=<rgba 0 0.8218 1 1>}
  NamedRerouteDeclaration(None=StaticSwitchParameter(
    True=TextureCoordinate()
    False=MaterialFunctionCall!(UseLocalSpace, WorldPosition, NormalOverride) {materialFunction=/Game/Environments/Apollo/Sets/Skyfire/Materials/MaterialFunctions/MF_WorldAlignedUVs_HardEdge}.Result
   ) {parameterName=UseStandardUVCoords}) {name=CheapWorldAlignedUVs, nodeColor=<rgba 0 0.7656 1 1>}
  n27 = NamedRerouteUsage()
  n28 = VectorParameter() {group=Noise, parameterName=NoiseA_TilePan, defaultValue=<rgba 1000 1000 0.0113 0.0097>}
  n29 = AppendVector(A=@n28.R, B=@n28.R)
  n30 = StaticSwitchParameter(True=Multiply(A=@n27, B=@n29), False=Divide(A=@n27, B=@n29)) {parameterName=UseStandardUVCoords}
  n31 = AppendVector(A=@n28.B, B=@n28.B)
  n32 = NamedRerouteUsage()
  n33 = VectorParameter() {group=Noise, parameterName=NoiseB_TilePan, defaultValue=<rgba -500 -500 0.007 0.0065>}
  n34 = AppendVector(A=@n33.R, B=@n33.R)
  n35 = StaticSwitchParameter(True=Multiply(A=@n32, B=@n34), False=Divide(A=@n32, B=@n34)) {parameterName=UseStandardUVCoords}
  n36 = AppendVector(A=@n33.B, B=@n33.B)
  n37 = Add(
   A=TextureSample(UVs=Panner(Coordinate=@n30, Speed=@n31) {bFractionalPart=true}) {texture=/Game/Global/Textures/Default/PointSplatter/T_Tiling_Dust}.R
   B=TextureSample(UVs=Panner(Coordinate=@n35, Speed=@n36) {bFractionalPart=true}) {texture=&T_Tiling_Dust}.R
  )
  n38 = ScalarParameter() {defaultValue=1, group=Noise, parameterName=NoiseBaseScale}
  n39 = Power(
   Base=Multiply(A=@n37, B=ScalarParameter() {defaultValue=2, group=Noise, parameterName=NoiseBrightScale})
   Exp=ScalarParameter() {defaultValue=5.5, group=Noise, parameterName=NoiseExponent}
  )
  n40 = NamedRerouteUsage()
  n41 = VectorParameter() {group=Noise, parameterName=NoiseC_TilePan, defaultValue=<rgba 800 800 0.0875 -0.077>}
  n42 = AppendVector(A=@n41.R, B=@n41.R)
  n43 = Multiply(A=@n40, B=@n42)
  n44 = Divide(A=@n40, B=@n42)
  n45 = TextureSample(UVs=Panner(Coordinate=StaticSwitchParameter(True=@n43, False=@n44) {parameterName=UseStandardUVCoords}, Speed=AppendVector(A=@n41.B, B=@n41.B)) {bFractionalPart=true}) {texture=&T_ShieldCaustics}
  n46 = NamedRerouteUsage()
  n47 = VectorParameter() {group=Noise, parameterName=NoiseD_TilePan, defaultValue=<rgba -700 -700 0.075 0.0605>}
  n48 = AppendVector(A=@n47.R, B=@n47.R)
  n49 = Multiply(A=@n46, B=@n48)
  n50 = Divide(A=@n46, B=@n48)
  n51 = TextureSample(UVs=Panner(Coordinate=StaticSwitchParameter(True=@n49, False=@n50) {parameterName=UseStandardUVCoords}, Speed=AppendVector(A=@n47.B, B=@n47.B)) {bFractionalPart=true}) {texture=&T_ShieldCaustics}
  n52 = NamedRerouteDeclaration(None=Add(
    A=NamedRerouteDeclaration(None=@n45.R) {name=NoiseC, nodeColor=<rgba 0.7109 1 0 1>}
    B=NamedRerouteDeclaration(None=@n51.R) {name=NoiseD, nodeColor=<rgba 0.3359 1 0 1>}
   )) {name=NoiseC+D, nodeColor=<rgba 1 0 0.6562 1>}
  NamedRerouteDeclaration(None=Add(A=Multiply(A=@n37, B=@n38), B=Multiply(A=@n39, B=@n52))) {name=Noise_ABCD, nodeColor=<rgba 0 1 0.2734 1>}
  n53 = NamedRerouteUsage()
  n54 = ScalarParameter() {defaultValue=1, group=Grain, parameterName=GrainScaleV}
  n55 = ScalarParameter() {defaultValue=1, group=Grain, parameterName=GrainScaleU}
  n56 = ScalarParameter() {defaultValue=35, group=Grain, parameterName=GrainDivisorV}
  n57 = ScalarParameter() {defaultValue=35, group=Grain, parameterName=GrainDivisorU}
  n58 = StaticSwitchParameter(True=Multiply(A=@n53, B=AppendVector(A=@n54, B=@n55)), False=Divide(A=@n53, B=AppendVector(A=@n56, B=@n57))) {parameterName=UseStandardUVCoords}
  n59 = Add(
   A=Time()
   B=Subtract(A=TextureSample(UVs=@n58) {samplerType=SAMPLERTYPE_LinearGrayscale, texture=/Engine/EngineMaterials/Good64x64TilingNoiseHighFreq}.R) {constB=0.5}
  )
  NamedRerouteDeclaration(None=Multiply(A=Sine(None=@n59) {period=0.44}, B=ScalarParameter() {defaultValue=0.2, group=Grain, parameterName=GrainIntensity})) {name=Grain, nodeColor=<rgba 1 0 0.4922 1>}
  n60 = ComponentMask(None=Multiply(A=ReflectionVectorWS(), B=ScalarParameter() {defaultValue=2, group=Chroma, parameterName=ChromaReflectionVectorScale})) {g=true, r=true}
  n61 = Constant() {r=0.5}
  n62 = CurveAtlasRowParameter(CurveTime=Frac(None=DotProduct(A=@n60, B=@n61))) {
   atlas=&SkyfireCurveAtlas
   curve=/Game/Environments/Apollo/Sets/Skyfire/Curves/C_Skyfire_Rainbow
   defaultValue=17
   group=Chroma
   parameterName=ChromaColorCurve
  }
  NamedRerouteDeclaration(None=Multiply(A=@n62, B=VectorParameter() {group=Chroma, parameterName=ChromaIntensity, defaultValue=<rgba 1.8 0.2 1.8 0>}.RGB)) {name=Chroma, nodeColor=<rgba 0.5 0 1 1>}
  NamedRerouteUsage()
  NamedRerouteUsage()
  NamedRerouteUsage()
 ]
}
