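# Parent material for Skyfire motion tubes: a dark, semi-metallic Default Lit surface with animated emissive bands.
# A panned UV coordinate runs through a sine and SphereMask to form sweeps, multiplied by a time-driven pulse offset per block by a tiled pattern texture.
# A static switch takes emissive color from either a curve atlas row via custom primitive data or a vector parameter.
# Two blended detail normals layer over the base normal, alongside adjustable ambient occlusion.
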Material /Game/Environments/Apollo/Sets/Skyfire/Materials/Parents/M_Skyfire_MotionTube_Parent
graph {
 properties {pixelDepthOffsetMode=PDOM_Legacy, usedShadingModels=Default Lit}
 nodes [
  n1 = Panner(
   Coordinate=NamedRerouteUsage()
   Speed=AppendVector(
    A=ScalarParameter() {defaultValue=12, group=World Mask, parameterName=WorldMaskPanU}
    B=ScalarParameter() {defaultValue=9, group=World Mask, parameterName=WorldMaskPanV}
   )
  )
  n2 = ScalarParameter() {defaultValue=550, group=World Mask, parameterName=WorldMaskDivisor}
  NamedRerouteDeclaration(None=TextureSample(UVs=Divide(A=@n1, B=@n2)) {texture=/Game/Global/Textures/Default/Noise/T_Tiling_Noise_01}.R) {name=WorldMask, nodeColor=<rgba 0 1 0.6016 1>}
  n3 = TextureSampleParameter2D() {parameterName=Masks, samplerType=SAMPLERTYPE_LinearColor, texture=/Game/Environments/Apollo/Sets/Skyfire/Textures/T_Skyfire_EmptyMask}
  NamedRerouteDeclaration(None=@n3.RGBA) {name=MasksRGBA, nodeColor=<rgba 0 1 0.9531 1>}
  n4 = TextureSampleParameter2D() {group=Normals, parameterName=Normal, samplerType=SAMPLERTYPE_Normal, texture=&FlatNormalBlue}
  n5 = TextureSampleParameter2D(UVs=Divide(A=NamedRerouteUsage(), B=ScalarParameter() {defaultValue=100, group=Normals, parameterName=DetailNormalBTileDivisor})) {
   group=Normals
   parameterName=DetailNormalB
   samplerType=SAMPLERTYPE_Normal
   texture=/Game/Global/Textures/Default/Tileable_Patterns/T_Skyfire_DarkGlass_DetailNormal
  }
  n6 = Multiply(A=@n5.RGB, B=VectorParameter() {group=Normals, parameterName=DetailNormalStrengthB, defaultValue=<rgba 0.15 0.15 1 0>}.RGB)
  n7 = TextureSampleParameter2D(UVs=NamedRerouteUsage()) {
   group=Normals
   parameterName=DetailNormalA
   samplerType=SAMPLERTYPE_Normal
   texture=/Game/Environments/Apollo/Sets/Skyfire/Textures/T_Skyfire_BlockNormals
  }
  n8 = Multiply(
   A=NamedRerouteDeclaration(None=@n7.RGB) {name=DetailNormal, nodeColor=<rgba 0 0.0391 1 1>}
   B=VectorParameter() {group=Normals, parameterName=DetailNormalStrengthA, defaultValue=<rgba 0.1 0.1 1 0>}.RGB
  )
  n9 = NamedRerouteUsage()
  NamedRerouteDeclaration(None=MaterialFunctionCall(BaseNormal=@n4.RGB, AdditionalNormal=LinearInterpolate(A=@n6, B=@n8, Alpha=@n9)) {materialFunction=&BlendAngleCorrectedNormals}.Result) {name=FinalNormals, nodeColor=<rgba 0 0.7188 1 1>}
  n10 = ScalarParameter() {defaultValue=1, group=Ambient Occlusion, parameterName=AO_Strength}
  n11 = NamedRerouteUsage()
  NamedRerouteDeclaration(None=LinearInterpolate(B=NamedRerouteUsage(), Alpha=Max(A=@n10, B=@n11)) {constA=1}) {name=AmbientOcclusion, nodeColor=<rgba 1 0.7969 0 1>}
  NamedRerouteDeclaration(None=@n3.RGB) {name=MasksRGB, nodeColor=<rgba 0 0.0391 1 1>}
  NamedRerouteDeclaration(None=MaterialFunctionCall!(UseLocalSpace, WorldPosition, NormalOverride) {materialFunction=/Game/Environments/Apollo/Sets/Skyfire/Materials/MaterialFunctions/MF_WorldAlignedUVs_HardEdge}.Result) {name=CheapWorldAlignedUVs, nodeColor=<rgba 0 0.7656 1 1>}
  NamedRerouteDeclaration(None=@n3.R) {name=MasksR_Glow, nodeColor=<rgba 0 0.3672 1 1>}
  NamedRerouteDeclaration(None=@n3.G) {name=MasksG_Height, nodeColor=<rgba 0 1 0 1>}
  NamedRerouteDeclaration(None=@n3.B) {name=MasksB_AO, nodeColor=<rgba 0 0 1 1>}
  NamedRerouteDeclaration(None=@n3.A) {name=MasksA_CurvatureFine, nodeColor=<rgba 1 1 1 1>}
  n12 = Panner(
   Coordinate=TextureCoordinate()
   Speed=AppendVector(A=ScalarParameter() {defaultValue=0.3, parameterName=EmissiveSweepPanU}, B=ScalarParameter() {parameterName=EmissiveSweepPanV})
  ) {speedX=0.1}
  n13 = ConstantBiasScale(None=Sine(None=ComponentMask(None=@n12) {r=true}))
  n14 = ScalarParameter() {parameterName=EmissiveSweepCenter}
  n15 = ScalarParameter() {defaultValue=0.4, parameterName=EmissiveSweepRadius}
  n16 = ScalarParameter() {parameterName=EmissiveSweepHardness}
  n17 = CurveAtlasRowParameter() {
   atlas=/Game/Environments/Apollo/Sets/Skyfire/Curves/SkyfireCurveAtlas
   bUseCustomPrimitiveData=true
   curve=/Game/Environments/Apollo/Sets/Skyfire/Curves/C_Skyfire_Black
   defaultValue=9
   group=Emissive
   parameterName=ColorCurveCPD
   primitiveDataIndex=5
  }
  n18 = StaticSwitchParameter(True=@n17, False=VectorParameter() {parameterName=EmissiveColor, defaultValue=<rgba 0.9625 0.1117 1 0>}.RGB) {group=Color, parameterName=EmissiveUsesCPD}
  n19 = TextureCoordinate()
  n20 = AppendVector(A=ScalarParameter() {defaultValue=4, parameterName=TechPatternScaleU}, B=ScalarParameter() {defaultValue=4, parameterName=TechPatternScaleV})
  n21 = TextureSampleParameter2D(UVs=NamedRerouteDeclaration(None=Multiply(A=@n19, B=@n20)) {name=TechPatternUVs, nodeColor=<rgba 0.8516 1 0 1>}) {
   group=TexturePacks
   parameterName=PackA
   samplerType=SAMPLERTYPE_LinearColor
   texture=/Game/Environments/Apollo/Sets/Skyfire/Textures/T_Skyfire_PackA
  }
  n22 = Add(A=Divide(A=Time(), B=ScalarParameter() {defaultValue=10, group=Core Glow, parameterName=BlocksEmissiveTimeScale}), B=@n21.G)
  n23 = Add(A=ConstantBiasScale(None=Sine(None=@n22) {period=0.6}), B=ScalarParameter() {defaultValue=0.6, parameterName=EmissiveBase})
  NamedRerouteDeclaration(None=Multiply(A=SphereMask(A=@n13, B=@n14, Radius=@n15, Hardness=@n16), B=Multiply(A=@n18, B=@n23))) {name=Emissive, nodeColor=<rgba 0 0.0625 1 1>}
  NamedRerouteUsage()
  NamedRerouteUsage()
  NamedRerouteUsage()
  ScalarParameter() {defaultValue=0.4, parameterName=Metallic}
  ScalarParameter() {defaultValue=0.8, parameterName=Specular}
  ScalarParameter() {defaultValue=0.2, parameterName=Roughness}
  VectorParameter() {parameterName=Base Color, defaultValue=<rgba 0.033 0.023 0.0319 0>}
 ]
}
