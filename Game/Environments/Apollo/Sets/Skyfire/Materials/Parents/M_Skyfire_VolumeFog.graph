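# M_Skyfire_VolumeFog is an additive volume-domain parent material for animated volumetric fog in the Skyfire set.
# A tunable gradient across the bounding-box UVW, inverted and raised to a power, fades the fog along a chosen axis.
# A world-space Voronoi volume texture pans over time and is remapped to signed noise that modulates that gradient.
# The result is scaled by an Extinction parameter, alongside an Albedo color multiplied by EmissiveScale.
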Material /Game/Environments/Apollo/Sets/Skyfire/Materials/Parents/M_Skyfire_VolumeFog
graph {
 properties {blendMode=BLEND_Additive, materialDomain=MD_Volume, pixelDepthOffsetMode=PDOM_Legacy, usedShadingModels=Default Lit}
 nodes [
  Multiply(A=VectorParameter() {parameterName=Albedo, defaultValue=<rgba 0.8217 0.6698 1 0>}.RGB, B=ScalarParameter() {parameterName=EmissiveScale})
  n1 = MaterialFunctionCall() {materialFunction=&BoundingBoxBased_0-1_UVW}
  n2 = VectorParameter() {parameterName=FadeGradientAxis, defaultValue=<rgba 0 0.3 1 0>}
  n3 = Power(Base=OneMinus(None=DotProduct(A=@n1.RGB, B=@n2.RGB)), Exp=ScalarParameter() {defaultValue=2, parameterName=HeightFalloffGradient})
  n4 = Time()
  n5 = VectorParameter() {parameterName=VolumeFogPanSpeedWorldSpace, defaultValue=<rgba 50 0 22 0>}
  n6 = Divide(
   A=Add(A=WorldPosition().XYZ, B=Multiply(A=@n4, B=@n5.RGB))
   B=ScalarParameter() {defaultValue=2000, parameterName=VolumeTextureWorldSpaceDivisor}
  ) {constB=500}
  n7 = ConstantBiasScale(None=Power(
    Base=TextureSampleParameterVolume!(UVs=@n6, Apply View MipBias) {parameterName=Param, samplerType=SAMPLERTYPE_LinearColor, texture=/Game/Effects/FogVolume/VT_Voronoi_LOW}.R
    Exp=ScalarParameter() {defaultValue=2.1, parameterName=VolumeTextureExponent}
   )) {bias=-0.5, scale=2}
  Multiply(A=ScalarParameter() {defaultValue=0.3, parameterName=Extinction}, B=Add(A=@n3, B=Multiply(A=@n3, B=@n7)))
 ]
}
